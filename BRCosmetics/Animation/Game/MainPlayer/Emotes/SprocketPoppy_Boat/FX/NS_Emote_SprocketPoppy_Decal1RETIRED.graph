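# Niagara effect for the SprocketPoppy boat emote, named as retired, that flashes a short-lived blue-cyan glow and decal-like sprite.
# Two burst emitters share a brief lifetime: ParticleDecal draws a camera-offset sprite with alpha fading in and out, while NE_BookGlow adds fixed-facing sprites and a curve-pulsed light.
# Time-of-day inputs tint the decal sprite and dim the glow at night.
# NE_BookGlow uses two spawn groups to vary a dynamic material parameter.
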NiagaraSystem /BRCosmetics/Animation/Game/MainPlayer/Emotes/SprocketPoppy_Boat/FX/NS_Emote_SprocketPoppy_Decal1RETIRED
dependencies {
 usedDataInterfaces [&NiagaraDataInterfaceColorCurve, &NiagaraDataInterfaceCurve, &NiagaraDataInterfaceArrayFloat]
 usedDynamicInputs [&ConvertVectorToPosition]
 usedModules [&SystemState]
}
emitters {
 NE_BookGlow {
  emitterUpdateScript [
   &EmitterState(Life Cycle Mode=::NewEnumerator1, Loop Behavior=::NewEnumerator1, Loop Duration=1, UseLoopDelay=false)
   &SpawnBurst_Instantaneous(Use Spawn Probability=false, UseLoopCountLimit=false)
  ]
  particleSpawnScript [
   &FN_SetSpawnGroup(Number of Spawn Groups=2)
   &InitializeParticle(
    Color=<rgba 0.6472 0.8976 1 0.0242>
    Lifetime=0.6
    Sprite Rotation Angle=0.1156
    Sprite Rotation Mode=::NewEnumerator1
    Sprite Size Mode=::NewEnumerator3
    Uniform Sprite Size=&SelectFloatFromArray(Direct Array Index=$Particles.SpawnGroup, Float Selection Array=DataInterface() {propertyValues {FloatData [100,100]}})
    UsePositionOffset=false
    Write Material Random=true
   )
  ]
  particleUpdateScript [
   &ParticleState()
   &ScaleColor(
    Scale RGB=&Multiply_VectorByFloat(Vector=<1 1 1>, Float=&FNTimeOfDayScalar(Night Scale=0.3, Morning Scale=1, Evening Scale=1))
    ScaleA=true
    ScaleRGB=true
   )
   &FN_Light_Attributes(
    Color=$Particles.Color
    Create Color Attribute=true
    Create Exponent Attribute=true
    Create Radius Attribute=true
    Exponent=6
    Radius=200
    Scale Color=&FloatFromCurve(
     FloatCurve=DataInterface() {propertyValues {Curve {keys [
         {interpMode=RCIM_Cubic, tangentMode=RCTM_User, time=0, value=0}
         {
          arriveTangent=-0.0501
          arriveTangentWeight=0.348
          interpMode=RCIM_Cubic
          leaveTangent=-0.0501
          leaveTangentWeight=0.348
          tangentMode=RCTM_User
          tangentWeightMode=RCTWM_WeightedBoth
          time=0.5
          value=1
         }
         {tangentMode=RCTM_User, time=1, value=0}
        ]}}}
     Scale Curve=0.3
    )
   )
   &DynamicMaterialParameters(
    Index 1 Param 1=&MaskFloatBySpawnGroup(Masked Float=-2, Passthrough Float=1)
    Module.Index1Float0Write=true
    Module.Index1Float1Write=true
    Module.Index1Float2Write=true
    Module.Index1Float3Write=true
    Write Parameter Index 0=::NewEnumerator0
    Write Parameter Index 1=::NewEnumerator1
   )
   /SetVariables(Particles.DecalSize:Vector2f=<60 60>) {isSetParameters=true}
   /SetVariables(Particles.SpriteFacing=<0 -1 0>) {isSetParameters=true}
  ]
  rendererClasses [&NiagaraLightRendererProperties, &NiagaraSpriteRendererProperties]
  renderers [
   {rendererClass=&NiagaraLightRendererProperties, rendererIndex=0}
   {rendererClass=&NiagaraSpriteRendererProperties, rendererIndex=1}
  ]
 }
 ParticleDecal {
  emitterUpdateScript [
   &EmitterState(Life Cycle Mode=::NewEnumerator1, Loop Behavior=::NewEnumerator1, Loop Duration=1, UseLoopDelay=false)
   &SpawnBurst_Instantaneous(Use Spawn Probability=false, UseLoopCountLimit=false)
  ]
  particleSpawnScript [
   &InitializeParticle(
    Color=&FNTimeOfDayColor(
     Night Color=<rgba 0 0.5124 1.4 1>
     Morning Color=<rgba 0.4826 1.5713 2 1>
     Day Color=<rgba 0.4826 1.5713 2 1>
     Evening Color=<rgba 0.4826 1.5713 2 1>
    )
    Lifetime=0.6
    Mesh Scale Mode=::NewEnumerator3
    Mesh Uniform Scale=2
    Position Offset=<0 0 100>
    Sprite Size Mode=::NewEnumerator3
    Uniform Sprite Size=40
    UsePositionOffset=false
    Write Material Random=true
   )
   &CameraOffset(Camera Offset Amount=30)
   &DynamicMaterialParameters(
    Index 0 Param 1=1
    Index 0 Param 3=1
    Module.Index0Param0Write=true
    Module.Index0Param1Write=true
    Module.Index0Param2Write=true
    Module.Index0Param3Write=true
   )
  ]
  particleUpdateScript [
   &ParticleState()
   &ScaleColor(
    Scale Alpha=&FloatFromCurve(FloatCurve=DataInterface() {propertyValues {Curve {keys [
         {interpMode=RCIM_Cubic, tangentMode=RCTM_User, time=0, value=0}
         {
          arriveTangent=0.1116
          arriveTangentWeight=0.4713
          interpMode=RCIM_Cubic
          leaveTangent=-0.3115
          leaveTangentWeight=0.2344
          tangentMode=RCTM_Break
          tangentWeightMode=RCTWM_WeightedBoth
          time=0.5
          value=1
         }
         {arriveTangent=-7.7147, interpMode=RCIM_Cubic, leaveTangent=-7.7147, tangentMode=RCTM_User, time=1, value=0}
        ]}}})
    ScaleA=true
    ScaleRGB=false
   )
  ]
  rendererClasses [&NiagaraSpriteRendererProperties]
  renderers [{rendererClass=&NiagaraSpriteRendererProperties}]
 }
}
summary {
 systemName=NS_Emote_SprocketPoppy_Decal1RETIRED
 emitters [
  {emitterName=ParticleDecal, rendererClasses [&NiagaraSpriteRendererProperties]}
  {emitterName=NE_BookGlow, rendererClasses [&NiagaraLightRendererProperties, &NiagaraSpriteRendererProperties]}
 ]
 userVariables []
}
