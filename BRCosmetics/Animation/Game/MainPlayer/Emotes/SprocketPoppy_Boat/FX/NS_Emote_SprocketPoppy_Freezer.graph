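# Niagara system for the freezer frost effect in the SprocketPoppy Boat player emote.
# It holds one emitter, SnowConeFrostLWE, whose name suggests a lightweight emitter; no renderers are listed for it.
# A system spawn assignment and dynamic inputs for time-of-day color and vector/color conversion accompany the standard SystemState module.
# A User.FreezeLifetime float is exposed, presumably to control how long the frost lasts.
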NiagaraSystem /BRCosmetics/Animation/Game/MainPlayer/Emotes/SprocketPoppy_Boat/FX/NS_Emote_SprocketPoppy_Freezer
dependencies {
 usedDataInterfaces []
 usedDynamicInputs [
  /Niagara/DynamicInputs/LinearColor/MakeLinearColorFromVector4
  &MakeVector4_FromLinearColor
  &FNTimeOfDayColor
 ]
 usedModules [
  &self:SystemSpawnScript.SystemScriptSource.SystemScriptGraph.NiagaraNodeAssignment_0.SetVariables_#1
  &SystemState
 ]
}
emitters {SnowConeFrostLWE {rendererClasses [], renderers []}}
summary {
 systemName=NS_Emote_SprocketPoppy_Freezer
 emitters [{emitterName=SnowConeFrostLWE, rendererClasses []}]
 userVariables [User.FreezeLifetime {type=NiagaraFloat, defaultValue=1, description=""}]
}
