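# Niagara ice-blast effect for the SprocketPoppy Boat emote, marked RETIRED in its name.
# Six emitters layer frost sprites, a flickering hand glow, blast sprites, spike meshes, falling shards and speed-stretched motion lines.
# Frost reads hand and cone sockets and uses the distance between them to scale drag, while inheriting source movement.
# Time-of-day color and brightness adjust several layers, with front-end and platform modifiers on the glow.
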NiagaraSystem /BRCosmetics/Animation/Game/MainPlayer/Emotes/SprocketPoppy_Boat/FX/NS_Emote_SprocketPoppy_Ice1RETIRED
dependencies {
 usedDataInterfaces [
  &NiagaraDataInterfaceCurve
  &NiagaraDataInterfaceVectorCurve
  &NiagaraDataInterfaceColorCurve
  &NiagaraDataInterfacePlatformSet
  &NiagaraDataInterfaceArrayFloat
  &NiagaraDataInterfaceMeshRendererInfo
  &NiagaraDataInterfaceSocketReader
 ]
 usedDynamicInputs [&Vector2DFromFloat]
 usedModules [
  &self:SystemSpawnScript.SystemScriptSource.SystemScriptGraph.NiagaraNodeAssignment_0.SetVariables_#1
  &SystemState
 ]
}
emitters {
 FrostSprites {
  emitterUpdateScript [
   &EmitterState(Life Cycle Mode=::NewEnumerator1, Loop Behavior=::NewEnumerator1, Loop Duration=0.3, UseLoopDelay=false)
   &SpawnRate(SpawnRate=17, Use Spawn Probability=false) {enabled=false, module=SpawnRate001}
   &SpawnRate(SpawnRate=30, Use Spawn Probability=false)
  ]
  particleSpawnScript [
   &InitializeParticle(
    Color=<rgba 0.6667 0.8769 1 0.25>
    Lifetime=&RandomRangeFloat(Minimum=0.7, Maximum=1)
    Sprite Rotation Mode=::NewEnumerator2
    Sprite Size Mode=::NewEnumerator1
    Sprite UV Mode=::NewEnumerator3
    Uniform Sprite Size Max=40
    Uniform Sprite Size Min=60
    UsePositionOffset=false
    Write Material Random=true
   )
   &AddVelocity(
    Rotation Coordinate Space=::Simulation
    Velocity=&TransformVector(Vector=<0 1 0>, Destination Space=::World)
    Velocity Speed Scale=&Multiply_Float(A=&RandomRangeFloat(Minimum=380, Maximum=400), B=1)
   )
   &ApplyInitialForces()
   /SetVariables() {isSetParameters=true}
   &CameraOffset(Camera Offset Amount=15)
  ]
  particleUpdateScript [
   &SocketLocation(
    Orientation Sampling=::NewEnumerator1
    Position Sampling=::NewEnumerator1
    SamplePosOffset=false
    Socket Reader=$System.HandLocation
    Socket Sampling Mode=::NewEnumerator3
    VelocitySpeedLimit=false
   )
   &SocketLocation(
    Orientation Sampling=::NewEnumerator1
    Position Sampling=::NewEnumerator1
    SamplePosOffset=false
    Socket Reader=$System.ConeLocation
    Socket Sampling Mode=::NewEnumerator3
    VelocitySpeedLimit=false
   ) {module=SocketLocation001}
   /SetVariables(Particles.HandToCone:NiagaraFloat=&VectorLength(Vector=&Divide_Vector(
      A:Vector3f=&Subtract_Position(A=$Output.SocketLocation001.SampledPosition, B=$Output.SocketLocation.SampledPosition)
      B:Vector3f=<100 100 100>
     ))) {isSetParameters=true}
   &ParticleState()
   &ScaleSpriteSize(Uniform Curve Sprite Scale=DataInterface() {propertyValues {Curve {keys [
        {arriveTangent=0.4427, interpMode=RCIM_Cubic, leaveTangent=0.4427, tangentMode=RCTM_User, time=0, value=0.2}
        {arriveTangent=1.0864, interpMode=RCIM_Cubic, leaveTangent=1.0864, tangentMode=RCTM_User, time=1, value=1}
       ]}}})
   &ScaleColor(
    Scale Alpha=&Multiply_Float(
     A=&FloatFromCurve(
      FloatCurve=DataInterface() {propertyValues {Curve {keys [
          {arriveTangent=39.813, interpMode=RCIM_Cubic, leaveTangent=39.813, tangentMode=RCTM_User, time=0, value=0}
          {
           arriveTangent=-3.0471
           interpMode=RCIM_Cubic
           leaveTangent=-3.0471
           leaveTangentWeight=0.647
           tangentMode=RCTM_User
           time=0.0241
           value=1.9918
          }
          {arriveTangent=-3.2131, interpMode=RCIM_Cubic, leaveTangent=-3.2131, tangentMode=RCTM_User, time=1, value=0}
         ]}}}
      Scale Curve=&FloatFromCurve(
       FloatCurve=DataInterface() {propertyValues {Curve {keys [
           {arriveTangent=10.3561, interpMode=RCIM_Cubic, leaveTangent=10.3561, tangentMode=RCTM_User, time=0, value=0}
           {arriveTangent=1.1179, interpMode=RCIM_Cubic, leaveTangent=1.1179, tangentMode=RCTM_User, time=0.0949, value=0.6228}
           {arriveTangent=0.0224, interpMode=RCIM_Cubic, leaveTangent=0.0224, tangentMode=RCTM_User, time=1, value=1}
          ]}}}
       CurveIndex=$Emitter.Age
      )
     )
     B=&FloatFromCurve(FloatCurve=DataInterface() {propertyValues {Curve {keys [
          {tangentMode=RCTM_User, time=0, value=1}
          {arriveTangent=-0.0405, interpMode=RCIM_Cubic, leaveTangent=-0.0405, tangentMode=RCTM_User, time=0.2975, value=1}
          {arriveTangent=-8.7355, interpMode=RCIM_Cubic, leaveTangent=-8.7355, tangentMode=RCTM_User, time=0.5148, value=0}
         ]}}})
    )
    ScaleA=true
    ScaleRGB=true
   )
   /SetVariables(Particles.SubImageIndex=&FloatFromCurve(FloatCurve=DataInterface() {propertyValues {Curve {keys [<at 0 30>,<at 1 64>]}}})) {isSetParameters=true}
   &ScaleColor(
    Scale Mode=::NewEnumerator1
    Scale RGBA=&MakeVector4_FromLinearColor(LinearColor=&FNTimeOfDayColor(
      Night Color=<rgba 0.539 0.6941 1 0.3>
      Morning Color=<rgba 0.82 0.8878 1 0.3>
      Day Color=<rgba 1 1 1 0.6>
      Evening Color=<rgba 1 0.8495 0.8351 0.3>
     ))
    ScaleRGBA=true
   ) {module=ScaleColor001}
   &DynamicMaterialParameters(
    Index 0 Param 1=3
    Index 0 Param 2=1
    Index 0 Param 3=0.5
    Module.Index0Param0Write=true
    Module.Index0Param1Write=true
    Module.Index0Param2Write=true
    Module.Index0Param3Write=true
   )
   &Drag(
    Drag=&FloatFromCurve(
     FloatCurve=DataInterface() {propertyValues {Curve {keys [
         {interpMode=RCIM_Cubic, tangentMode=RCTM_User, time=0, value=0}
         {arriveTangent=22.0021, interpMode=RCIM_Cubic, leaveTangent=22.0021, tangentMode=RCTM_User, time=0.06, value=1}
        ]}}}
     Scale Curve=&Multiply_Float(A=5, B=&Subtract_Float(A=2.5, B=$Particles.HandToCone))
    )
    Use Linear Drag=true
    Use Rotational Drag=true
   ) {module=Drag002}
   &InheritSourceMovement(
    Applied Position Falloff Distance=1000
    Applied Position Scale=&Multiply_VectorByFloat(
     Vector=<1 1 0>
     Float=&FloatFromCurve(
      FloatCurve=DataInterface() {propertyValues {Curve {keys [
          {arriveTangent=-3.2201, interpMode=RCIM_Cubic, leaveTangent=-3.2201, tangentMode=RCTM_User, time=0.1817, value=1}
          {arriveTangent=-0.3194, interpMode=RCIM_Cubic, leaveTangent=-0.3194, tangentMode=RCTM_User, time=1, value=0}
         ]}}}
      CurveIndex=$Emitter.Age
     )
    )
    ApplyForce=false
    ApplyPosition=true
    ApplyPositionFalloff=false
   )
   &SolveForcesAndVelocity(Clamp Velocity=false, Limit Acceleration=false)
  ]
  rendererClasses [&NiagaraSpriteRendererProperties]
  renderers [{rendererClass=&NiagaraSpriteRendererProperties}]
 }
 HandGlow {
  emitterUpdateScript [
   &EmitterState(Inactive Response=::NewEnumerator1, Life Cycle Mode=::NewEnumerator1, Loop Behavior=::NewEnumerator1, Loop Duration=0.5, UseLoopDelay=false)
   &SpawnBurst_Instantaneous(Use Spawn Probability=false, UseLoopCountLimit=false)
  ]
  particleSpawnScript [
   &InitializeParticle(
    Color=<rgba 0.15 0.15 0.15 1.5>
    Lifetime=0.48
    Mesh Scale=&Multiply_VectorByFloat(Vector=<1.7 1 1>, Float=0.02)
    Mesh Scale Mode=::NewEnumerator4
    Sprite Rotation Mode=::NewEnumerator1
    Sprite Size Mode=::NewEnumerator3
    Uniform Sprite Size=20
    UsePositionOffset=false
    Write Material Random=true
   )
   &InitialMeshOrientation(Mesh Axis To Orient=<0 -1 0>, Mesh Orientation Mode=::NewEnumerator2, Rotation=<0.6323 -0.7 0>, Use Rotation Vector=true)
   &CameraOffset(Camera Offset Amount=5)
   &AddVelocity(Velocity=<0 0.0001 0>)
   &SpriteFacingAndAlignment(Sprite Facing=<0 1 0>, WriteAlignment=false, WriteFacing=true)
  ]
  particleUpdateScript [
   &ParticleState()
   &ScaleMeshSize(Scale Factor=&Multiply_Vector(
     A=&ScaleFloatByCurve(
      InitialValue:Vector3f=<1 1 1>
      ScaleCurve:NiagaraDataInterfaceVectorCurve=DataInterface() {propertyValues {
        XCurve {keys [<at 0 0.25>,<at 0.1 1>,<at 0.9 1>,<at 1 0>]}
        YCurve {keys [<at 0 1.25>,<at 0.1 2.25>,<at 0.9 2.25>,<at 1 0>]}
        ZCurve {keys [<at 0 1.25>,<at 0.1 2.25>,<at 0.9 2.25>,<at 1 0>]}
       }}
      CurveIndex:NiagaraFloat=$Particles.NormalizedAge
     )
     B=&Divide_Vector(A:Vector3f=<1 1 1>, B:Vector3f=$Engine.Owner.Scale)
    ))
   &ScaleSpriteSize(Uniform Curve Sprite Scale=DataInterface() {propertyValues {Curve {keys [
        <at 0 0>
        <at 0.1 1>
        <at 0.1512 0.7174>
        <at 0.2132 1>
        <at 0.2641 0.7539>
        <at 0.299 1>
        <at 0.3563 0.7037>
        <at 0.4151 1>
        <at 0.4564 0.8017>
        <at 0.5041 1>
        <at 0.5502 0.7561>
        <at 0.5836 1>
        <at 0.6551 0.7653>
        <at 0.6981 1>
        <at 0.7521 0.7721>
        <at 0.7871 1>
        <at 0.8475 0.7106>
        <at 0.9 1>
        {tangentMode=RCTM_User, time=1, value=0}
       ]}}})
   /SetVariables(Particles.Size002:Vector2f=&Multiply_Vector2D(
     A=$Particles.SpriteSize
     B=&Multiply_Vector2D_ByFloat(
      Float=&FloatFromCurve(FloatCurve=DataInterface() {propertyValues {Curve {keys [
           {interpMode=RCIM_Cubic, tangentMode=RCTM_User, time=0, value=1}
           {arriveTangent=-1.234, interpMode=RCIM_Cubic, leaveTangent=-1.234, tangentMode=RCTM_User, time=0.1, value=1.6328}
           {arriveTangentWeight=0.0997, interpMode=RCIM_Cubic, leaveTangentWeight=0.3488, tangentMode=RCTM_User, time=0.1688, value=1}
           {interpMode=RCIM_Cubic, tangentMode=RCTM_User, time=1, value=1}
          ]}}})
      Vector2D=<0.5 1>
     )
    )) {isSetParameters=true}
   /SetVariables(Particles.Size003:Vector2f=&Multiply_Vector2D(A=&Multiply_Vector2D_ByFloat(Float=20, Vector2D=<1 2>), B=<1 1>)) {isSetParameters=true}
   &ScaleColor(
    Scale Alpha=&Multiply_Float(
     A=&FloatFromCurve(
      FloatCurve=DataInterface() {propertyValues {Curve {keys [
          {arriveTangent=4.4475, time=0, value=0}
          {arriveTangent=5.3518, leaveTangent=-3.1774, time=0.1, value=1.2}
          {arriveTangent=-1.4972, leaveTangent=-1.4972, time=0.1523, value=0.7}
          {arriveTangent=-0.1667, leaveTangent=-0.1667, time=0.2052, value=1.1711}
          <at 0.258 0.7713>
          <at 0.2926 1.1775>
          {arriveTangent=-0.25, leaveTangent=-0.25, time=0.3608, value=0.689}
          <at 0.3865 1.193>
          <at 0.4263 0.8642>
          <at 0.4931 1.1191>
          {arriveTangent=-0.25, leaveTangent=-0.25, time=0.5428, value=0.7548}
          <at 0.5846 1.1063>
          <at 0.6571 0.8076>
          <at 0.6976 1.1685>
          {arriveTangent=0.1667, leaveTangent=0.1667, time=0.7542, value=0.7713}
          <at 0.7906 1.1138>
          <at 0.8535 0.8807>
          {arriveTangent=-3.3333, leaveTangent=-3.3333, time=0.9, value=1.2}
          <at 1 0>
         ]}}}
      Scale Curve=&FNTimeOfDayScalar(Night Scale=0.6, Morning Scale=0.7, Day Scale=0.8, Evening Scale=0.6)
     )
     B=&FN_FrontEndFloatModifier(Gameplay=1, Front End=0.3)
    )
    Scale RGB=&MakeVectorFromLinearColorRGB(Color=&ColorFromCurve(ColorCurve=DataInterface() {propertyValues {
        AlphaCurve {keys []}
        BlueCurve {keys [<at 0.7887 0.437 RCIM_Cubic>,<at 0.9801 0.1562 RCIM_Cubic>]}
        GreenCurve {keys [<at 0.7887 0.4375 RCIM_Cubic>,<at 0.9801 0.1445 RCIM_Cubic>]}
        RedCurve {keys [<at 0.7887 0.3349 RCIM_Cubic>,<at 0.9801 0.1274 RCIM_Cubic>]}
       }}))
    ScaleA=true
    ScaleRGB=true
   )
   /SetVariables(Particles.color002:LinearColor=&MakeLinearColorFromVectorAndFloat(
     "Vector (RGB)"=&MakeVectorFromLinearColorRGB(Color=<rgba 1.7807 2.8114 3 1>)
     "Float (Alpha)"=&FloatFromCurve(
      FloatCurve=DataInterface() {propertyValues {Curve {keys [<at 0 1>,<at 0.1 2>,<at 0.2005 1.7767>,<at 0.306 1.9927>,<at 0.3889 1.7548>,<at 0.5026 2.0076>,<at 0.5773 1.7477>,<at 0.6601 1.9791>,<at 0.7283 1.7134>,<at 0.8209 1.9848>,<at 1 0>]}}}
      Scale Curve=&FN_PickFloatDependOnPlatform(TargetPlatformActive=5, TargetPlatformNoActive=1)
     )
    )) {isSetParameters=true}
   /SetVariables(Particles.Color003:LinearColor=&MakeLinearColorFromVectorAndFloat(
     "Vector (RGB)"=&MakeVectorFromLinearColorRGB(Color=<rgba 0.7068 0.9585 1 1>)
     "Float (Alpha)"=&Multiply_Float(
      A=&FloatFromCurve(
       FloatCurve=DataInterface() {propertyValues {Curve {keys [
           {arriveTangent=0.3376, interpMode=RCIM_Cubic, leaveTangent=0.3376, tangentMode=RCTM_User, time=0.1324, value=0}
           <at 0.3 1>
           <at 0.4 0.7>
           <at 0.5 0.95>
           <at 0.6 0.55>
           <at 0.7 1>
           <at 0.8 0.55>
           <at 0.9 0.8>
           <at 1 0>
          ]}}}
       Scale Curve=&FNTimeOfDayScalar(Night Scale=0.4, Morning Scale=0.3, Day Scale=0.8, Evening Scale=0.5)
      )
      B=&Multiply_Float(
       A=&FN_FrontEndFloatModifier(Gameplay=0.5, Front End=0.15)
       B=&FN_PickFloatDependOnPlatform(TargetPlatformActive=1, TargetPlatformNoActive=0.4)
      )
     )
    )) {isSetParameters=true}
   /SetVariables(Particles.MeshColor:LinearColor=&MakeLinearColorFromVectorAndFloat(
     "Vector (RGB)"=&MakeVectorFromLinearColorRGB(Color=&ColorFromCurve(ColorCurve=DataInterface() {propertyValues {
         AlphaCurve {keys []}
         BlueCurve {keys [
           {arriveTangent=-1.8762, interpMode=RCIM_Cubic, leaveTangent=-1.8762, tangentMode=RCTM_User, time=0.1298, value=1.9554}
           {interpMode=RCIM_Cubic, leaveTangent=0.2132, time=0.3441, value=1.9833}
          ]}
         GreenCurve {keys [
           {arriveTangent=-0.4331, interpMode=RCIM_Cubic, leaveTangent=-0.4331, tangentMode=RCTM_User, time=0.1298, value=2}
           {interpMode=RCIM_Cubic, leaveTangent=2.0123, time=0.3441, value=2}
          ]}
         RedCurve {keys [
           {arriveTangent=-4.577, interpMode=RCIM_Cubic, leaveTangent=-4.577, tangentMode=RCTM_User, time=0.1298, value=1.8924}
           {interpMode=RCIM_Cubic, leaveTangent=-2.7231, time=0.3441, value=1.6652}
          ]}
        }}))
     "Float (Alpha)"=&Multiply_Float(
      A=&FloatFromCurve(
       FloatCurve=DataInterface() {propertyValues {Curve {keys [
           {arriveTangent=-0.9, time=0, value=0}
           {arriveTangent=-0.0998, leaveTangent=-0.0998, time=0.05, value=1}
           {arriveTangent=-1.0872, leaveTangent=-1.0872, time=0.1804, value=0.7734}
           <at 0.2985 0.937>
           <at 0.4531 0.6834>
           <at 0.5813 0.9005>
           <at 0.7511 0.6392>
           <at 0.8564 0.8>
           {arriveTangent=-1.5912, leaveTangent=-1.5912, time=1, value=0}
          ]}}}
       Scale Curve=&FN_FrontEndFloatModifier(Gameplay=1, Front End=0.5)
      )
      B=&FNTimeOfDayScalar(Night Scale=0.7, Morning Scale=0.5, Evening Scale=0.6)
     )
    )) {isSetParameters=true}
   &SolveForcesAndVelocity(Clamp Velocity=false, Limit Acceleration=false)
   /SetVariables(Particles.Vector 4:Vector4f=<1 1 1 1>) {isSetParameters=true}
  ]
  rendererClasses [&NiagaraSpriteRendererProperties]
  renderers [
   {rendererClass=&NiagaraSpriteRendererProperties}
   {rendererClass=&NiagaraSpriteRendererProperties, rendererIndex=1}
   {rendererClass=&NiagaraSpriteRendererProperties, rendererIndex=2}
  ]
 }
 IceBlast {
  emitterUpdateScript [
   &EmitterState(Life Cycle Mode=::NewEnumerator1, Loop Behavior=::NewEnumerator1, Loop Delay=0.05, Loop Duration=0.35, UseLoopDelay=false)
   &SpawnRate(SpawnRate=60, Use Spawn Probability=false)
  ]
  particleSpawnScript [
   &InitializeParticle(
    Color=&RandomRangeLinearColor(Minimum=<rgba 0.3851 0.6754 0.9479 0.8>, Maximum=<rgba 0.6933 0.8281 0.8802 0.8>)
    Lifetime Max=0.2
    Lifetime Min=0.3
    Lifetime Mode=::NewEnumerator1
    Position Offset=<0 -1.0927 0>
    Sprite Rotation Angle=&SelectFloatFromArray(
     Direct Array Index=&ReturnExecIndex()
     Direct Array Mode=::NewEnumerator1
     Float Selection Array=DataInterface() {propertyValues {FloatData [0.2,-0.2]}}
    )
    Sprite Rotation Mode=::NewEnumerator1
    Sprite Size Max=<10 30>
    Sprite Size Min=<15 35>
    Sprite Size Mode=::NewEnumerator2
    UsePositionOffset=true
    Write Material Random=true
   )
   &CameraOffset(Camera Offset Amount=5)
   &SpriteFacingAndAlignment(WriteAlignment=true, WriteFacing=false)
  ]
  particleUpdateScript [
   /SetVariables(Particles.SpriteRotation:NiagaraFloat=&SelectFloatFromArray(
     Direct Array Index=&ReturnExecIndex()
     Direct Array Mode=::NewEnumerator1
     Float Selection Array=DataInterface() {propertyValues {FloatData [0,4,-4,0,2,-2,0,1,-1,0,3,-3,0,5,-5,0]}}
    )) {isSetParameters=true}
   &ParticleState()
   &ScaleColor(
    Scale Alpha=&FloatFromCurve(FloatCurve=DataInterface() {propertyValues {Curve {keys [
         {arriveTangent=-0.0888, interpMode=RCIM_Cubic, leaveTangent=-0.0888, tangentMode=RCTM_User, time=0, value=0.5}
         {arriveTangent=0.12, interpMode=RCIM_Cubic, leaveTangent=0.12, tangentMode=RCTM_User, time=0.25, value=0.9847}
         {arriveTangent=-0.0405, interpMode=RCIM_Cubic, leaveTangent=-0.0405, tangentMode=RCTM_User, time=1, value=0}
        ]}}})
    Scale RGB=&VectorFromFloat(Value=1.1)
    ScaleA=true
    ScaleRGB=true
   )
   &ScaleSpriteSize(
    Scale Factor=&Multiply_Vector2D_ByFloat(
     Float=&FloatFromCurve(FloatCurve=DataInterface() {propertyValues {Curve {keys [
          {
           arriveTangent=-0.0024
           arriveTangentWeight=0.5946
           interpMode=RCIM_Cubic
           leaveTangentWeight=0.5946
           tangentMode=RCTM_User
           time=0.0677
           value=0
          }
          {arriveTangent=4.5684, interpMode=RCIM_Cubic, leaveTangent=4.5684, tangentMode=RCTM_User, time=0.7, value=1}
         ]}}})
     Vector2D=<1 2.5>
    )
    Scale Sprite Size Mode=::NewEnumerator2
   ) {module=ScaleSpriteSize003}
   &DynamicMaterialParameters(
    Index 0 Param 1=&Add_Float(
     A=&RandomRangeFloat(Minimum=0, Maximum=1)
     B=&FloatFromCurve(
      FloatCurve=DataInterface() {propertyValues {Curve {keys [
          {arriveTangent=2.7533, interpMode=RCIM_Cubic, leaveTangent=1.8958, tangentMode=RCTM_User, time=0, value=1}
          {arriveTangent=0.8292, interpMode=RCIM_Cubic, leaveTangent=0.8292, tangentMode=RCTM_User, time=0.321, value=1.3453}
          {arriveTangent=0.0529, interpMode=RCIM_Cubic, leaveTangent=0.2021, tangentMode=RCTM_User, time=1, value=1.5}
         ]}}}
      CurveIndex=&Add_Float(A=&Multiply_Float(A=1, B=0.5), B=$Particles.NormalizedAge)
     )
    )
    Index 0 Param 2=&FloatFromCurve(FloatCurve=DataInterface() {propertyValues {Curve {keys [
         {arriveTangent=0.0453, interpMode=RCIM_Cubic, tangentMode=RCTM_Break, time=0, value=-1}
         {arriveTangent=0.0604, interpMode=RCIM_Cubic, leaveTangent=0.0604, tangentMode=RCTM_User, time=1, value=-1}
        ]}}})
    Index 0 Param 3=&FloatFromCurve(
     FloatCurve=DataInterface() {propertyValues {Curve {keys [{tangentMode=RCTM_User, time=0, value=0}, {tangentMode=RCTM_User, time=1, value=0}]}}}
     Scale Curve=0
    )
    Index 0 Param 4=&FloatFromCurve(FloatCurve=DataInterface() {propertyValues {Curve {keys [<at 0.5 1>,<at 1 0>]}}})
    Index 1 Param 1=&FloatFromCurve(
     FloatCurve=DataInterface() {propertyValues {Curve {keys [
         {interpMode=RCIM_Cubic, leaveTangent=4.7088, tangentMode=RCTM_Break, time=0, value=0}
         {arriveTangent=4.2257, interpMode=RCIM_Cubic, leaveTangent=4.2257, tangentMode=RCTM_User, time=0.25, value=1}
        ]}}}
     Scale Curve=&RandomRangeFloat(Minimum=1.5, Maximum=1.5)
    )
    Index 1 Param 2=&FloatFromCurve(
     FloatCurve=DataInterface() {propertyValues {Curve {keys [<at 0 0.2>, {leaveTangent=0.9066, time=1, value=2}]}}}
     Scale Curve=&RandomRangeFloat(Minimum=0.5, Maximum=0.6)
    )
    Index 1 Param 3=&FloatFromCurve(FloatCurve=DataInterface() {propertyValues {Curve {keys [
         {interpMode=RCIM_Cubic, leaveTangent=6.3917, tangentMode=RCTM_Break, time=0, value=0.2}
         {arriveTangent=1.7549, interpMode=RCIM_Cubic, leaveTangent=10.6767, tangentMode=RCTM_User, time=0.25, value=1}
        ]}}})
    Index 1 Param 4=&FloatFromCurve(
     FloatCurve=DataInterface() {propertyValues {Curve {keys [{arriveTangent=-0.136, interpMode=RCIM_Cubic, leaveTangent=-0.136, tangentMode=RCTM_User, time=0, value=1}, <at 1 0.95>]}}}
     Scale Curve=&RandomRangeFloat(Minimum=1.1, Maximum=1.4)
    )
    Index 2 Param 1=&FloatFromCurve(FloatCurve=DataInterface() {propertyValues {Curve {keys [
         {interpMode=RCIM_Cubic, leaveTangent=4.2485, tangentMode=RCTM_Break, time=0, value=0}
         {arriveTangent=1.0548, interpMode=RCIM_Cubic, leaveTangent=1.0548, tangentMode=RCTM_User, time=0.17, value=0.5715}
         <at 1 1>
        ]}}})
    Index 2 Param 2=&FloatFromCurve(
     FloatCurve=DataInterface() {propertyValues {Curve {keys [
         {arriveTangent=1.9056, interpMode=RCIM_Cubic, leaveTangent=1.9056, tangentMode=RCTM_User, time=0, value=0.2}
         {arriveTangent=0.0259, interpMode=RCIM_Cubic, leaveTangent=0.0259, tangentMode=RCTM_User, time=1, value=1.4}
        ]}}}
     Scale Curve=&RandomRangeFloat(Minimum=0.2, Maximum=0.5)
    )
    Index 2 Param 3=&FloatFromCurve(FloatCurve=DataInterface() {propertyValues {Curve {keys [
         {interpMode=RCIM_Cubic, leaveTangent=2.3064, tangentMode=RCTM_Break, time=0, value=1}
         {interpMode=RCIM_Cubic, leaveTangent=11.1233, tangentMode=RCTM_Break, time=1, value=2}
        ]}}})
    Index 2 Param 4=&FloatFromCurve(
     FloatCurve=DataInterface() {propertyValues {Curve {keys [
         {arriveTangent=-0.6867, interpMode=RCIM_Cubic, leaveTangent=-0.6867, tangentMode=RCTM_User, time=0, value=1}
         {interpMode=RCIM_Cubic, leaveTangent=-1.2956, tangentMode=RCTM_Break, time=1, value=0.75}
        ]}}}
     Scale Curve=1.6
    )
    Module.Param0WriteEnabled=true
    Module.Param0WriteEnabled1=false
    Module.Param0WriteEnabled2=true
    Module.Param1WriteEnabled=false
    Module.Param1WriteEnabled1=false
    Module.Param1WriteEnabled2=true
    Module.Param2WriteEnabled=true
    Module.Param2WriteEnabled1=false
    Module.Param2WriteEnabled2=true
    Module.Param3WriteEnabled=true
    Module.Param3WriteEnabled1=false
    Module.Param3WriteEnabled2=true
    Write Parameter Index 1=true
    Write Parameter Index 2=true
   ) {module=DynamicMaterialParameters001}
   &SolveForcesAndVelocity(Clamp Velocity=false, Limit Acceleration=false)
   /SetVariables(Particles.Linear Color=&RandomRangeLinearColor(
     Minimum=&ColorFromCurve(ColorCurve=DataInterface() {propertyValues {
        AlphaCurve {keys [<at 0 0.6665>, {arriveTangent=-0.6665, interpMode=RCIM_Cubic, leaveTangent=-0.6665, time=0.0818, value=2}, <at 1 0 RCIM_Cubic>]}
        BlueCurve {keys [<at 0 0.949 RCIM_Cubic>]}
        GreenCurve {keys [<at 0 0.6745 RCIM_Cubic>]}
        RedCurve {keys [<at 0 0.3843 RCIM_Cubic>]}
       }})
     Maximum=&ColorFromCurve(ColorCurve=DataInterface() {propertyValues {
        AlphaCurve {keys [<at 0 0.6665>, {arriveTangent=-0.6665, interpMode=RCIM_Cubic, leaveTangent=-0.6665, time=0.1479, value=2}, <at 1 0 RCIM_Cubic>]}
        BlueCurve {keys [<at 0 0.8784 RCIM_Cubic>]}
        GreenCurve {keys [<at 0 0.8275 RCIM_Cubic>]}
        RedCurve {keys [<at 0 0.6941 RCIM_Cubic>]}
       }})
    )) {isSetParameters=true}
   /SetVariables(Particles.LowMediumSize:Vector2f=&Multiply_Vector2D_ByFloat(
     Float=&FloatFromCurve(FloatCurve=DataInterface() {propertyValues {Curve {keys [
          {
           arriveTangent=3.1042
           arriveTangentWeight=0.5946
           interpMode=RCIM_Cubic
           leaveTangent=3.1042
           leaveTangentWeight=0.5946
           tangentMode=RCTM_User
           time=0.0677
           value=0
          }
          {arriveTangent=0.3198, interpMode=RCIM_Cubic, leaveTangent=0.3198, tangentMode=RCTM_User, time=0.7, value=1}
         ]}}})
     Vector2D=&Multiply_Vector2D(A=<1 2.5>, B=&RandomRangeVector2D(Minimum=<15 26.0958>, Maximum=<10 28.9994>))
    )) {isSetParameters=true}
  ]
  rendererClasses [&NiagaraSpriteRendererProperties]
  renderers [
   {rendererClass=&NiagaraSpriteRendererProperties}
   {rendererClass=&NiagaraSpriteRendererProperties, rendererIndex=1}
  ]
 }
 IceSpikes {
  emitterUpdateScript [
   &EmitterState(Life Cycle Mode=::NewEnumerator1, Loop Behavior=::NewEnumerator1, Loop Delay=0.05, Loop Duration=0.05, UseLoopDelay=true)
   &SpawnRate(Spawn Probability=0.67, SpawnRate=160, Use Spawn Probability=false)
  ]
  particleSpawnScript [
   &InitializeParticle(
    Color=&RandomRangeLinearColor(Minimum=<rgba 0.3851 0.6754 0.9479 0.8>, Maximum=<rgba 0.188 0.545 0.8802 0.8>)
    Lifetime=&RandomRangeFloat(Minimum=0.15, Maximum=0.2)
    Mesh Scale Mode=::NewEnumerator1
    Mesh Uniform Scale Max=0.5
    Mesh Uniform Scale Min=0.3
    Sprite Rotation Mode=::NewEnumerator2
    Sprite Size=<100 200>
    Sprite Size Mode=::NewEnumerator4
    UsePositionOffset=false
    Write Material Random=true
   ) {module=InitializeParticle001}
   /SetVariables(Particles.Rand:NiagaraFloat=&UniformAOrBFloat(A=0, Evaluation Type=::NewEnumerator0, Recalculate Random Each Loop=true, Override Seed=false)) {isSetParameters=true}
   &InitialMeshOrientation(Mesh Orientation Mode=::NewEnumerator1, Rotation=<0 0 0.2457>, Rotation Coordinate Space=::NewEnumerator0, Use Rotation Vector=true)
  ]
  particleUpdateScript [
   &ParticleState()
   &DynamicMaterialParameters(
    Index 0 Param 1=$Particles.NormalizedAge
    Index 0 Param 2=&RandomRangeFloat(Minimum=1, Maximum=1.5)
    Index 0 Param 3=&RandomRangeFloat(Minimum=0, Maximum=1)
    Index 0 Param 4=&FloatFromCurve(
     FloatCurve=DataInterface() {propertyValues {Curve {keys [
         {arriveTangent=2.9585, interpMode=RCIM_Cubic, leaveTangent=2.9585, tangentMode=RCTM_User, time=0, value=0}
         {arriveTangent=-0.0405, interpMode=RCIM_Cubic, leaveTangent=-0.0405, tangentMode=RCTM_User, time=1, value=1}
        ]}}}
     Scale Curve=&RandomRangeFloat(Minimum=0.4, Maximum=0.6)
    )
    Module.Param0WriteEnabled=true
    Module.Param1WriteEnabled=true
    Module.Param2WriteEnabled=true
    Module.Param3WriteEnabled=true
   )
   &ScaleMeshSize(Scale Factor=&VectorFromFloat(Value=&FloatFromCurve(FloatCurve=DataInterface() {propertyValues {Curve {keys [
          {arriveTangent=3.844, interpMode=RCIM_Cubic, leaveTangent=3.844, tangentMode=RCTM_User, time=0, value=0}
          {arriveTangent=0.1649, interpMode=RCIM_Cubic, leaveTangent=0.1649, tangentMode=RCTM_User, time=0.75, value=1}
         ]}}})))
   &ScaleColor(
    Scale Alpha=&FloatFromCurve(
     FloatCurve=DataInterface() {propertyValues {Curve {keys [{interpMode=RCIM_Cubic, tangentMode=RCTM_User, time=0.1, value=1}, {interpMode=RCIM_Cubic, tangentMode=RCTM_User, time=1, value=0}]}}}
     Scale Curve=2
    )
    Scale RGB=&VectorFromFloat(Value=5)
    ScaleA=true
    ScaleRGB=true
   )
   &SolveForcesAndVelocity(Clamp Velocity=false, Limit Acceleration=false)
  ]
  rendererClasses [&NiagaraMeshRendererProperties]
  renderers [{rendererClass=&NiagaraMeshRendererProperties, rendererIndex=0}]
 }
 Ice_Shards {
  emitterUpdateScript [
   &EmitterState(Life Cycle Mode=::NewEnumerator1, Loop Behavior=::NewEnumerator1, Loop Duration=1, UseLoopDelay=false)
   &SpawnRate(SpawnRate=10, Use Spawn Probability=false)
  ]
  particleSpawnScript [
   &InitializeParticle(
    Lifetime Max=0.4
    Lifetime Min=0.092
    Lifetime Mode=::NewEnumerator1
    Mesh Scale=&Multiply_VectorByFloat(Vector=&RandomRangeVector(Minimum=<0.15 0.2 0.14>, Maximum=<0.5 0.4 0.3>), Float=0.3)
    Mesh Scale Mode=::NewEnumerator4
    UsePositionOffset=false
    Write Material Random=true
   )
   &ShapeLocation(
    MaskSpawnGroup=false
    Offset Mode=::NewEnumerator0
    Override Seed=false
    Rotation Mode=::NewEnumerator1
    Sphere Radius=5
    Sphere Surface Distribution=1
    UseOwnerScale=true
   )
   &InitialMeshOrientation(Mesh Axis To Orient=<0 0 1>, Mesh Orientation Mode=::NewEnumerator2, Use Rotation Vector=true)
   &AddVelocity(Velocity=&RandomRangeVector(Minimum=<-60 90 -60>, Maximum=<60 150 60>))
  ]
  particleUpdateScript [
   &ParticleState()
   &ScaleColor(
    Scale Alpha=&FloatFromCurve(FloatCurve=DataInterface() {propertyValues {Curve {keys [<at 0.5 1 RCIM_Cubic>,<at 1.0011 0.002 RCIM_Cubic>]}}}, Scale Curve=0.22)
    Scale RGB=&MakeVectorFromLinearColorRGB(Color=&FN_Select_Color_from_Time_Of_Day(Morning=<rgba 0.3208 0.5045 0.8 1>, Day=<rgba 0.3125 0.5928 1 1>, Evening=<rgba 0.3542 0.6556 1 1>, Night=<rgba 0.2161 0.2929 0.5 1>))
    ScaleA=true
    ScaleRGB=true
   )
   &ScaleMeshSize(Scale Factor=&VectorFromFloat(Value=&FloatFromCurve(
      FloatCurve=DataInterface() {propertyValues {Curve {keys [
          {arriveTangent=9.0725, interpMode=RCIM_Cubic, leaveTangent=9.0725, tangentMode=RCTM_User, time=0, value=0}
          {arriveTangent=-2.5249, interpMode=RCIM_Cubic, leaveTangent=-2.5249, tangentMode=RCTM_User, time=0.1408, value=0.9847}
          {arriveTangent=-0.0405, interpMode=RCIM_Cubic, leaveTangent=-0.0405, tangentMode=RCTM_User, time=1, value=0}
         ]}}}
      Scale Curve=1.7
     )))
   &Drag(Use Linear Drag=true, Use Rotational Drag=true)
   &GravityForce(Gravity=&Multiply_Vector(
     A=<0 0 -475.6099>
     B=&VectorFromFloat(Value=&FloatFromCurve(FloatCurve=DataInterface() {propertyValues {Curve {keys [<at 0 0 RCIM_Cubic>,<at 1 1 RCIM_Cubic>]}}}))
    )) {module=GravityForce001}
   &SolveForcesAndVelocity(Clamp Velocity=false, Limit Acceleration=false)
  ]
  rendererClasses [&NiagaraMeshRendererProperties]
  renderers [{rendererClass=&NiagaraMeshRendererProperties, rendererIndex=0}]
 }
 MotionLines {
  emitterUpdateScript [
   &EmitterState(Life Cycle Mode=::NewEnumerator1, Loop Behavior=::NewEnumerator1, Loop Duration=1, UseLoopDelay=false)
   &SpawnRate(SpawnRate=15, Use Spawn Probability=false)
  ]
  particleSpawnScript [
   &InitializeParticle(
    Lifetime Max=0.3
    Lifetime Min=0.092
    Lifetime Mode=::NewEnumerator1
    Position Offset=<0 10 0>
    Sprite Size Mode=::NewEnumerator3
    Uniform Sprite Size=1.5
    UsePositionOffset=true
    Write Material Random=true
   )
   &ShapeLocation(
    MaskSpawnGroup=false
    Offset Mode=::NewEnumerator0
    Override Seed=false
    Rotation Mode=::NewEnumerator1
    Sphere Radius=5
    Sphere Surface Distribution=1
    UseOwnerScale=true
   ) {enabled=false}
   &InitialMeshOrientation(Mesh Axis To Orient=<0 0 1>, Mesh Orientation Mode=::NewEnumerator2, Use Rotation Vector=true)
   &AddVelocity(Velocity=&RandomRangeVector(Minimum=<-60 90 -60>, Maximum=<60 150 60>))
  ]
  particleUpdateScript [
   &ParticleState()
   &ScaleColor(
    Scale Alpha=&FloatFromCurve(FloatCurve=DataInterface() {propertyValues {Curve {keys [<at 0.5 1 RCIM_Cubic>,<at 1.0011 0.002 RCIM_Cubic>]}}}, Scale Curve=0.22)
    Scale RGB=&MakeVectorFromLinearColorRGB(Color=&FN_Select_Color_from_Time_Of_Day(Morning=<rgba 0.6536 0.9781 1.5 1>, Day=<rgba 0.6536 0.9781 1.5 1>, Evening=<rgba 0.6536 0.9781 1.5 1>, Night=<rgba 0.6536 0.9781 1.5 1>))
    ScaleA=true
    ScaleRGB=true
   )
   &Drag(Use Linear Drag=true, Use Rotational Drag=true)
   &GravityForce(Gravity=&Multiply_Vector(
     A=<0 0 -176.0709>
     B=&VectorFromFloat(Value=&FloatFromCurve(FloatCurve=DataInterface() {propertyValues {Curve {keys [<at 0 0 RCIM_Cubic>,<at 1 1 RCIM_Cubic>]}}}))
    )) {enabled=false, module=GravityForce001}
   &SolveForcesAndVelocity(Clamp Velocity=false, Limit Acceleration=false)
   &ScaleSpriteSizeBySpeed(Max Scale Factor=<0.5 6>, Min Scale Factor=<0.5 10>)
  ]
  rendererClasses [&NiagaraSpriteRendererProperties]
  renderers [{rendererClass=&NiagaraSpriteRendererProperties}]
 }
}
summary {
 systemName=NS_Emote_SprocketPoppy_Ice1RETIRED
 emitters [
  {emitterName=Ice_Shards, rendererClasses [&NiagaraMeshRendererProperties]}
  {emitterName=HandGlow, rendererClasses [&NiagaraSpriteRendererProperties]}
  {emitterName=FrostSprites, rendererClasses [&NiagaraSpriteRendererProperties]}
  {emitterName=IceBlast, rendererClasses [&NiagaraSpriteRendererProperties]}
  {emitterName=IceSpikes, rendererClasses [&NiagaraMeshRendererProperties]}
  {emitterName=MotionLines, rendererClasses [&NiagaraSpriteRendererProperties]}
 ]
 userVariables [
  User.bIsFrontEnd {type=NiagaraBool}
  User.bIsFrontEndPreview {type=NiagaraBool}
  User.bIsRightHand {type=NiagaraBool}
  User.FreezeLifetime {type=NiagaraFloat, defaultValue=1, description=""}
 ]
}
