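# Niagara system for the melting effect of the SprocketPoppy Boat emote, a main-player cosmetic.
# It holds a single emitter, SnowConeFrostLWE, whose name suggests a lightweight snow cone frost emitter.
# SystemState is the only listed module.
# The file shows no renderers, user variables, data interfaces or dynamic inputs.
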NiagaraSystem /BRCosmetics/Animation/Game/MainPlayer/Emotes/SprocketPoppy_Boat/FX/NS_Emote_SprocketPoppy_Melting
dependencies {usedDataInterfaces [], usedDynamicInputs [], usedModules [&SystemState]}
emitters {SnowConeFrostLWE {rendererClasses [], renderers []}}
summary {systemName=NS_Emote_SprocketPoppy_Melting, emitters [{emitterName=SnowConeFrostLWE, rendererClasses []}], userVariables []}
